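# Sound cue for a looping Alpine biome ambient bed, named BaseSmall and built on the Alpine night loop waves.
# A quality level node at the root routes its first two slots to the looping quad wave and its last two to the looping stereo (ST) wave.
# Overridden attenuation makes it non-spatialized with a box shape and logarithmic falloff, at reduced volume in an outdoor-ducking ambient sound class.
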SoundCue /Game/Sounds/Apollo/Biomes/Alpine/Biome_Alpine_Loop_BaseSmall_Cue
root=QualityLevel_0
nodes [
 {id=QualityLevel_0, children [WavePlayer_2,WavePlayer_2,WavePlayer_3,WavePlayer_3]}
 {id=WavePlayer_2, properties {bLooping=true, soundWaveAssetPtr=&/Biome_Alpine_Loop_Night_Quad}}
 {id=WavePlayer_3, properties {bLooping=true, soundWaveAssetPtr=&/Biome_Alpine_Loop_Night_ST}}
]
properties {
 attenuationSettings=&Fort_Ambient_Athena_Att
 bOverrideAttenuation=true
 maxDistance=6575.6592
 soundClassObject=/Game/Sounds/FortSoundClassesAndModes/Fort_Ambient_DuckOutdoors_Athena_PSM
 volumeMultiplier=0.401
 attenuationOverrides {
  attenuationShape=Box
  bSpatialize=false
  distanceAlgorithm=Logarithmic
  falloffDistance=1962.0334
  attenuationShapeExtents=<2509.1855 1425.9445 4613.626>
 }
}
