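# Sound cue for the looping base ambience of the Apollo Alpine biome, built from the biome's night loop recordings.
# A quality level node picks the looping quad wave for the first two levels and the looping stereo wave for the last two.
# Overridden box attenuation fades it by distance without spatialization, covering a large area.
# It plays at reduced volume through an outdoor-ducking ambient sound class.
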SoundCue /Game/Sounds/Apollo/Biomes/Alpine/Biome_Alpine_Loop_Base_Cue
root=QualityLevel_0
nodes [
 {id=QualityLevel_0, children [WavePlayer_2,WavePlayer_2,WavePlayer_3,WavePlayer_3]}
 {id=WavePlayer_2, properties {bLooping=true, soundWaveAssetPtr=&/Biome_Alpine_Loop_Night_Quad}}
 {id=WavePlayer_3, properties {bLooping=true, soundWaveAssetPtr=&/Biome_Alpine_Loop_Night_ST}}
]
properties {
 attenuationSettings=&Fort_Ambient_Athena_Att
 bOverrideAttenuation=true
 maxDistance=29042.5508
 soundClassObject=/Game/Sounds/FortSoundClassesAndModes/Fort_Ambient_DuckOutdoors_Athena_PSM
 volumeMultiplier=0.401
 attenuationOverrides {attenuationShape=Box, bSpatialize=false, falloffDistance=3877.4641, attenuationShapeExtents=<11833.043 25165.0859 2378.7717>}
}
