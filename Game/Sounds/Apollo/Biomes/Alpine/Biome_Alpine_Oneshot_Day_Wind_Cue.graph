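# Sound cue for randomized daytime wind one-shots in the Alpine biome's outdoor ambience, using tile-ambient one-shot attenuation.
# A quality-level root sends its first two levels to four Quad waves and its last two to four ST waves.
# Each set is picked by an equally weighted random node under a modulator that varies pitch and volume, with the ST branch modulated slightly quieter.
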SoundCue /Game/Sounds/Apollo/Biomes/Alpine/Biome_Alpine_Oneshot_Day_Wind_Cue
root=QualityLevel_0
nodes [
 {id=Modulator_0, children [Random_1], properties {pitchMax=1.1, pitchMin=0.9, volumeMax=0.9, volumeMin=0.8}}
 {id=Modulator_1, children [Random_0], properties {pitchMax=1.1, pitchMin=0.9, volumeMax=1.1, volumeMin=0.9}}
 {id=QualityLevel_0, children [Modulator_1,Modulator_1,Modulator_0,Modulator_0]}
 {id=Random_0, children [WavePlayer_1,WavePlayer_2,WavePlayer_3,WavePlayer_4], properties {weights [1,1,1,1]}}
 {id=Random_1, children [WavePlayer_5,WavePlayer_6,WavePlayer_7,WavePlayer_8], properties {weights [1,1,1,1]}}
 {id=WavePlayer_1, properties {soundWaveAssetPtr=&/Biome_Alpine_Oneshot_Day_Quad_01}}
 {id=WavePlayer_2, properties {soundWaveAssetPtr=&/Biome_Alpine_Oneshot_Day_Quad_02}}
 {id=WavePlayer_3, properties {soundWaveAssetPtr=&/Biome_Alpine_Oneshot_Day_Quad_03}}
 {id=WavePlayer_4, properties {soundWaveAssetPtr=&/Biome_Alpine_Oneshot_Day_Quad_04}}
 {id=WavePlayer_5, properties {soundWaveAssetPtr=&/Biome_Alpine_Oneshot_Day_ST_01}}
 {id=WavePlayer_6, properties {soundWaveAssetPtr=&/Biome_Alpine_Oneshot_Day_ST_02}}
 {id=WavePlayer_7, properties {soundWaveAssetPtr=&/Biome_Alpine_Oneshot_Day_ST_03}}
 {id=WavePlayer_8, properties {soundWaveAssetPtr=&/Biome_Alpine_Oneshot_Day_ST_04}}
]
properties {
 attenuationSettings=/Game/Sounds/Attenuation/Fort_TileAmbient_OneShot_Att
 maxDistance=8000
 soundClassObject=&Fort_Ambient_Outdoors_Athena
 volumeMultiplier=0.501
 preEffectBusSends [{}]
}
